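# Default Lit mesh-particle material for the head of the LemonCart dance grenade, textured with its diffuse, normal and mask maps.
# Dynamic parameter RotateSpeed sets a Z-axis rotation around the object position, with normals corrected by FixRotateAboutAxisNormals.
# Dynamic parameter WPO adds power-shaped deformation: XY offsets relative to a lowered pivot and a sine-eased Z offset, scaled by WPOMax.
# Particle color tints a Fresnel rim passed through MF_FX_EmissiveAdjustments.
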Material /ClydeCosmetics/Weapons/DanceGrenade/FX/M_LemonCart_DanceGrenade_Head
graph {
 properties {bUsedWithMeshParticles=true, bUsedWithNiagaraMeshParticles=true, pixelDepthOffsetMode=PDOM_Legacy, usedShadingModels=Default Lit}
 nodes [
  n1 = Constant3Vector() {constant=<rgba 0 0 1 0>}
  n2 = DynamicParameter() {defaultValue=<rgba 0 -0.25 1 1>, paramNames [WPO,RotateSpeed,Param3,Param4]}
  n3 = Multiply(A=@n2.RotateSpeed) {constB=-1}
  n4 = WorldPosition()
  n5 = Add(
   A=ObjectPositionWS()
   B=Multiply(A=Constant3Vector() {constant=<rgba 0 0 -0.25 0>}, B=ScalarParameter() {defaultValue=50, parameterName=Offset Local}) {constA=-0.25}
  )
  n6 = Multiply(A=ComponentMask(None=Subtract(A=@n4.XYZ, B=@n5)) {g=true, r=true}, B=Power(Base=@n2.WPO) {constExponent=5})
  n7 = Power(Base=@n2.WPO)
  n8 = Clamp(None=LinearInterpolate(Alpha=@n2.WPO) {constA=1, constB=0})
  n9 = Add(A=Multiply(A=@n8, B=LinearInterpolate(Alpha=@n8) {constA=1, constB=3})) {constB=0.25}
  n10 = Multiply(A=@n7, B=Multiply(A=@n7, B=Sine(None=@n9) {period=0.75}))
  Add(
   A=RotateAboutAxis(NormalizedRotationAxis=@n1, RotationAngle=@n3, PivotPoint=ObjectPositionWS(), Position=WorldPosition().XYZ)
   B=Multiply(A=VectorParameter() {parameterName=WPOMax, defaultValue=<rgba -1 -1 -100 1>}.RGB, B=AppendVector(A=@n6, B=@n10)) {constB=0}
  )
  n11 = MaterialFunctionCall(
   Rotation Axis=@n1
   Rotation Angle=@n3
   World Space Vertex Normals=Transform(None=TextureSample() {samplerType=SAMPLERTYPE_Normal, texture=/ClydeCosmetics/Weapons/DanceGrenade/Textures/T_Bomb_LemonCart_N}.RGB)
  ) {materialFunction=&FixRotateAboutAxisNormals}
  MaterialFunctionCall(Emissive=Multiply(A=Fresnel(Normal=@n11.New World Space Vertex Normal) {baseReflectFraction=0, exponent=4}, B=ParticleColor().RGB)) {materialFunction=&MF_FX_EmissiveAdjustments}
  TextureSample() {texture=/ClydeCosmetics/Weapons/DanceGrenade/Textures/T_Bomb_LemonCart_D}
  TextureSample() {samplerType=SAMPLERTYPE_Masks, texture=/ClydeCosmetics/Weapons/DanceGrenade/Textures/T_Bomb_LemonCart_S}
 ]
}
