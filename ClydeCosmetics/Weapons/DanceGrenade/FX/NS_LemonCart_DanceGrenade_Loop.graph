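# Niagara looping effect for the LemonCart dance grenade cosmetic: a glowing disco ball with intro flare, corona, sparkles, light rays and coloured smoke.
# Hand-keyed curves sampled on emitter age pulse the ball, corona, spline rays and Smoke_S material in a shared irregular rhythm.
# User.LocalClient lowers the ball and its surrounding effects for the local player.
# Smoke spawns from and follows Smoke_S particles via particle reads, while SmokePuffs spread outward separately.
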NiagaraSystem /ClydeCosmetics/Weapons/DanceGrenade/FX/NS_LemonCart_DanceGrenade_Loop
dependencies {
 usedDataInterfaces [
  &NiagaraDataInterfaceCurve
  &NiagaraDataInterfaceVectorCurve
  &NiagaraDataInterfaceVector2DCurve
  &NiagaraDataInterfaceParticleRead
 ]
 usedDynamicInputs [&VectorFromCurve]
 usedModules [&SystemState]
}
emitters {
 Corona_And_Spread {
  emitterUpdateScript [
   &EmitterState()
   &SpawnBurst_Instantaneous(Spawn Time=0.45, Use Spawn Probability=false, UseLoopCountLimit=true)
  ]
  particleSpawnScript [
   &InitializeParticle(
    Color=<rgba 14 6.0191 0 1>
    Lifetime=0.65
    Position Offset=&MakeCustomVectorFromBool(Bool=$User.LocalClient, True Vector=<0 0 205>, False Vector=<0 0 230>)
    Sprite Size Mode=::NewEnumerator3
    Uniform Sprite Size=68
    UsePositionOffset=true
    Write Material Random=true
   )
   &CameraOffset(Camera Offset Amount=-20)
  ]
  particleUpdateScript [
   &ParticleState(
    Kill Particles When Lifetime Has Elapsed=false
    Let Infinitely Lived Particles Die When Emitter Deactivates=true
    Loop Particles Lifetime=true
   )
   &DynamicMaterialParameters(
    Index 0 Param 1=1
    Index 0 Param 2=1
    Index 0 Param 3=&RandomRangeFloat(Minimum=0.5, Maximum=0.7)
    Index 0 Param 4=1
    Module.Index0Param0Write=true
    Module.Index0Param1Write=true
    Module.Index0Param2Write=true
    Module.Index0Param3Write=true
   )
   &ScaleSpriteSize(
    Scale Sprite Size Mode=::NewEnumerator0
    Uniform Scale Factor=&Add_Float(
     A=&FloatFromCurve(
      FloatCurve=DataInterface() {propertyValues {Curve {keys [
          {arriveTangent=-38.4417, time=0, value=0.05}
          <at 1.3 0.05>
          <at 1.35 1 RCIM_Cubic>
          <at 1.8 0.05 RCIM_Cubic>
          <at 1.85 1 RCIM_Cubic>
          <at 2.25 0.05 RCIM_Cubic>
          <at 2.3 1 RCIM_Cubic>
          {arriveTangent=-1.5625, leaveTangent=-1.5625, time=2.7, value=0.05}
          {arriveTangent=3.8, interpMode=RCIM_Cubic, leaveTangent=3.8, time=2.9, value=0.05}
          <at 2.95 1 RCIM_Cubic>
          {arriveTangent=-1.1111, leaveTangent=-1.1111, time=3.35, value=0.05}
          {arriveTangent=1.5833, interpMode=RCIM_Cubic, leaveTangent=1.5833, time=3.9, value=0.05}
          <at 3.95 1 RCIM_Cubic>
          <at 4.35 0.05 RCIM_Cubic>
          <at 4.4 1 RCIM_Cubic>
          {arriveTangent=-2, leaveTangent=-2, time=4.8, value=0.05}
          {arriveTangent=6.3333, interpMode=RCIM_Cubic, leaveTangent=6.3333, time=4.9, value=0.05}
          <at 4.95 1 RCIM_Cubic>
         ]}}}
      CurveIndex=$Emitter.Age
      Scale Curve=0.25
     )
     B=1
    )
   )
   &ScaleColor(
    Scale Alpha=1.2
    Scale RGB=&VectorFromFloat(Value=&Add_Float(
      A=&FloatFromCurve(
       FloatCurve=DataInterface() {propertyValues {Curve {keys [
           {arriveTangent=-38.4417, time=0, value=0.05}
           <at 1.3 0.05>
           <at 1.35 1 RCIM_Cubic>
           <at 1.8 0.05 RCIM_Cubic>
           <at 1.85 1 RCIM_Cubic>
           <at 2.25 0.05 RCIM_Cubic>
           <at 2.3 1 RCIM_Cubic>
           {arriveTangent=-1.5625, leaveTangent=-1.5625, time=2.7, value=0.05}
           {arriveTangent=3.8, interpMode=RCIM_Cubic, leaveTangent=3.8, time=2.9, value=0.05}
           <at 2.95 1 RCIM_Cubic>
           {arriveTangent=-1.1111, leaveTangent=-1.1111, time=3.35, value=0.05}
           {arriveTangent=1.5833, interpMode=RCIM_Cubic, leaveTangent=1.5833, time=3.9, value=0.05}
           <at 3.95 1 RCIM_Cubic>
           <at 4.35 0.05 RCIM_Cubic>
           <at 4.4 1 RCIM_Cubic>
           {arriveTangent=-2, leaveTangent=-2, time=4.8, value=0.05}
           {arriveTangent=6.3333, interpMode=RCIM_Cubic, leaveTangent=6.3333, time=4.9, value=0.05}
           <at 4.95 1 RCIM_Cubic>
          ]}}}
       CurveIndex=$Emitter.Age
       Scale Curve=2
      )
      B=1
     ))
    ScaleA=true
    ScaleRGB=true
   )
  ]
  rendererClasses [&NiagaraSpriteRendererProperties]
  renderers [{rendererClass=&NiagaraSpriteRendererProperties}]
 }
 DiscoBall {
  emitterUpdateScript [
   &EmitterState()
   &SpawnBurst_Instantaneous(Spawn Time=0.01, Use Spawn Probability=false, UseLoopCountLimit=true)
  ]
  particleSpawnScript [
   &InitializeParticle(
    Color=<rgba 8 5.9167 3 1>
    Lifetime=1.84
    Mesh Scale Mode=::NewEnumerator3
    Mesh Uniform Scale=0.625
    Position Offset=&MakeCustomVectorFromBool(Bool=$User.LocalClient, True Vector=<0 0 238>, False Vector=<0 0 263>)
    UsePositionOffset=true
    Write Material Random=true
   )
   &InitialMeshOrientation(Mesh Orientation Mode=::NewEnumerator1, Rotation=<0 0 0.5>, Use Rotation Vector=true)
  ]
  particleUpdateScript [
   &ParticleState(
    Kill Particles When Lifetime Has Elapsed=false
    Let Infinitely Lived Particles Die When Emitter Deactivates=true
    Loop Particles Lifetime=true
   )
   &ScaleColor(
    Scale RGB=&VectorFromFloat(Value=&Lerp_Float(
      A=&FloatFromCurve(FloatCurve=DataInterface() {propertyValues {Curve {keys [
           {arriveTangent=-38.4417, interpMode=RCIM_Cubic, time=0, value=15}
           {arriveTangent=-9.8311, interpMode=RCIM_Cubic, leaveTangent=-9.8311, tangentMode=RCTM_User, time=0.25, value=1}
           <at 0.495 0.2 RCIM_Cubic>
           <at 0.5 1 RCIM_Cubic>
           <at 0.745 0.2 RCIM_Cubic>
           <at 0.75 1 RCIM_Cubic>
           <at 1 0.2 RCIM_Cubic>
          ]}}})
      B=&FloatFromCurve(
       FloatCurve=DataInterface() {propertyValues {Curve {keys [
           {arriveTangent=-38.4417, interpMode=RCIM_Cubic, time=0, value=1}
           <at 0.354 0.2 RCIM_Cubic>
           <at 0.36 1 RCIM_Cubic>
           <at 0.708 0.2 RCIM_Cubic>
           <at 0.712 1 RCIM_Cubic>
           <at 1.062 0.2 RCIM_Cubic>
           <at 1.07 1 RCIM_Cubic>
           <at 1.416 0.2 RCIM_Cubic>
           <at 1.42 1 RCIM_Cubic>
           <at 1.77 0.2 RCIM_Cubic>
          ]}}}
       CurveIndex=$Particles.Age
       Scale Curve=50
      )
      Alpha=&MakeCustomFloatFromBool(Bool=&SetBoolByFloatComparison(A=$Emitter.Age, B=1.84), True Float=1, False Float=0)
     ))
    ScaleA=true
    ScaleRGB=true
   ) {enabled=false, module=ScaleColor001}
   &ScaleColor(
    Scale RGB=&VectorFromFloat(Value=&FloatFromCurve(
      FloatCurve=DataInterface() {propertyValues {Curve {keys [
          {arriveTangent=-38.4417, time=0, value=0}
          <at 1.3 0.05>
          <at 1.35 1 RCIM_Cubic>
          <at 1.8 0.05 RCIM_Cubic>
          <at 1.85 1 RCIM_Cubic>
          <at 2.25 0.05 RCIM_Cubic>
          <at 2.3 1 RCIM_Cubic>
          {arriveTangent=-1.5625, leaveTangent=-1.5625, time=2.7, value=0.05}
          {arriveTangent=3.8, interpMode=RCIM_Cubic, leaveTangent=3.8, time=2.9, value=0.05}
          <at 2.95 1 RCIM_Cubic>
          {arriveTangent=-1.1111, leaveTangent=-1.1111, time=3.35, value=0.05}
          {arriveTangent=1.5833, interpMode=RCIM_Cubic, leaveTangent=1.5833, time=3.9, value=0.05}
          <at 3.95 1 RCIM_Cubic>
          <at 4.35 0.05 RCIM_Cubic>
          <at 4.4 1 RCIM_Cubic>
          {arriveTangent=-2, leaveTangent=-2, time=4.8, value=0.05}
          {arriveTangent=6.3333, interpMode=RCIM_Cubic, leaveTangent=6.3333, time=4.9, value=0.05}
          <at 4.95 1 RCIM_Cubic>
         ]}}}
      CurveIndex=$Emitter.Age
      Scale Curve=10
     ))
    ScaleA=true
    ScaleRGB=true
   ) {enabled=false, module=ScaleColor002}
   &ScaleColor(
    Scale RGB=&VectorFromFloat(Value=&Add_Float(
      A=&FloatFromCurve(
       FloatCurve=DataInterface() {propertyValues {Curve {keys [
           {arriveTangent=-25.6702, interpMode=RCIM_Cubic, leaveTangent=-25.6702, tangentMode=RCTM_User, time=0, value=10}
           {arriveTangent=-0.937, interpMode=RCIM_Cubic, leaveTangent=0.4276, tangentMode=RCTM_Break, time=1, value=0.05}
           <at 1.3 0.05>
           <at 1.35 1 RCIM_Cubic>
           <at 1.8 0.05 RCIM_Cubic>
           <at 1.85 1 RCIM_Cubic>
           <at 2.25 0.05 RCIM_Cubic>
           <at 2.3 1 RCIM_Cubic>
           {arriveTangent=-1.5625, leaveTangent=-1.5625, time=2.7, value=0.05}
           {arriveTangent=3.8, interpMode=RCIM_Cubic, leaveTangent=3.8, time=2.9, value=0.05}
           <at 2.95 1 RCIM_Cubic>
           {arriveTangent=-1.1111, leaveTangent=-1.1111, time=3.35, value=0.05}
           {arriveTangent=1.5833, interpMode=RCIM_Cubic, leaveTangent=1.5833, time=3.9, value=0.05}
           <at 3.95 1 RCIM_Cubic>
           <at 4.35 0.05 RCIM_Cubic>
           <at 4.4 1 RCIM_Cubic>
           {arriveTangent=-2, leaveTangent=-2, time=4.8, value=0.05}
           {arriveTangent=6.3333, interpMode=RCIM_Cubic, leaveTangent=6.3333, time=4.9, value=0.05}
           <at 4.95 1 RCIM_Cubic>
          ]}}}
       CurveIndex=$Emitter.Age
       Scale Curve=5
      )
      B=&FloatFromCurve(
       FloatCurve=DataInterface() {propertyValues {Curve {keys [<at 0.95 0 RCIM_Cubic>,<at 1 1 RCIM_Cubic>,<at 2.1 0 RCIM_Cubic>,<at 2.15 1 RCIM_Cubic>,<at 3.05 0 RCIM_Cubic>,<at 3.1 1 RCIM_Cubic>,<at 5 0 RCIM_Cubic>]}}}
       CurveIndex=$Emitter.Age
       Scale Curve=0.5
      )
     ))
    ScaleA=true
    ScaleRGB=true
   )
   &DynamicMaterialParameters(
    Index 0 Param 1=&Lerp_Float(
     A=&FloatFromCurve(FloatCurve=DataInterface() {propertyValues {Curve {keys [{arriveTangent=-1, time=0, value=1}, <at 0.5 0>]}}})
     B=0
     Alpha=&MakeCustomFloatFromBool(Bool=&SetBoolByFloatComparison(A=$Emitter.Age, B=1.84), True Float=1, False Float=0)
    )
    Index 0 Param 2=&Lerp_Float(
     A=&FloatFromCurve(FloatCurve=DataInterface() {propertyValues {Curve {keys [
          <at 0.2 0 RCIM_Cubic>
          {arriveTangent=0.625, interpMode=RCIM_Cubic, leaveTangent=0.625, time=0.3, value=-0.1}
          {interpMode=RCIM_Cubic, leaveTangent=0.4445, time=1, value=0.5}
         ]}}})
     B=&FloatFromCurve(FloatCurve=DataInterface() {propertyValues {Curve {keys [{arriveTangent=0.5, time=0, value=0}, <at 1 0.5>]}}})
     Alpha=&MakeCustomFloatFromBool(Bool=&SetBoolByFloatComparison(A=$Emitter.Age, B=1.84), True Float=1, False Float=0)
    )
    Index 0 Param 3=1
    Index 0 Param 4=1
    Index 1 Param 1=&Lerp_Float(
     A=&FloatFromCurve(FloatCurve=DataInterface() {propertyValues {Curve {keys [{arriveTangent=-1, time=0, value=1}, <at 0.5 0>]}}})
     B=0
     Alpha=&MakeCustomFloatFromBool(Bool=&SetBoolByFloatComparison(A=$Emitter.Age, B=1.84), True Float=1, False Float=0)
    )
    Index 1 Param 2=&FloatFromCurve(
     FloatCurve=DataInterface() {propertyValues {Curve {keys [{arriveTangent=0.5, time=0, value=0}, <at 1 1>]}}}
     CurveIndex=$Emitter.NormalizedLoopAge
    )
    Module.Index0Param0Write=true
    Module.Index0Param1Write=true
    Module.Index0Param2Write=true
    Module.Index0Param3Write=true
    Module.Index1Param0Write=true
    Module.Index1Param1Write=true
    Module.Index1Param2Write=true
    Module.Index1Param3Write=true
    Write Parameter Index 1=true
   )
   &FN_Light_Attributes(
    Color=<rgba 2 1.5166 0.3125 1>
    Create Color Attribute=true
    Create Exponent Attribute=true
    Create Position Attribute=true
    Create Radius Attribute=true
    Exponent=1.5
    Position=<0 0 160>
    Radius=200
    Scale Color=4
   )
  ]
  rendererClasses [
   &NiagaraMeshRendererProperties
   &NiagaraLightRendererProperties
   &NiagaraComponentRendererProperties
  ]
  renderers [
   {rendererClass=&NiagaraMeshRendererProperties, rendererIndex=0}
   {rendererClass=&NiagaraLightRendererProperties, rendererIndex=1}
   {rendererClass=&NiagaraMeshRendererProperties, rendererIndex=2}
   {rendererClass=&NiagaraComponentRendererProperties, rendererIndex=3}
  ]
 }
 DiscoBall_Sparkles {
  emitterUpdateScript [
   &EmitterState(Delay First Loop Only=true, Life Cycle Mode=::NewEnumerator1, Loop Delay=0.45, Loop Duration=1, UseLoopDelay=true)
   &SpawnBurst_Instantaneous(
    Spawn Count=&RandomRangeInt(Minimum=2, Maximum=5)
    Use Spawn Probability=false
    UseLoopCountLimit=false
   )
   &SpawnRate(SpawnRate=12, Use Spawn Probability=false)
  ]
  particleSpawnScript [
   &InitializeParticle(
    Color=<rgba 2 1.3293 0.5938 1>
    Lifetime Max=0.3
    Lifetime Min=0.15
    Lifetime Mode=::NewEnumerator1
    Position Offset=&MakeCustomVectorFromBool(Bool=$User.LocalClient, True Vector=<0 0 205>, False Vector=<0 0 230>)
    Sprite Size Mode=::NewEnumerator1
    Uniform Sprite Size Max=30
    Uniform Sprite Size Min=18
    UsePositionOffset=true
    Write Material Random=true
   )
   &ShapeLocation(MaskSpawnGroup=false, Override Seed=false, Sphere Radius=20, Sphere Surface Distribution=1, UseOwnerScale=true)
   &CameraOffset(Camera Offset Amount=60)
  ]
  particleUpdateScript [
   &ParticleState()
   &ScaleSpriteSize(
    Scale Sprite Size Mode=::NewEnumerator0
    Uniform Scale Factor=&FloatFromCurve(FloatCurve=DataInterface() {propertyValues {Curve {keys [<at 0 0 RCIM_Cubic>, {arriveTangent=7.6708, interpMode=RCIM_Cubic, leaveTangent=2.1977, tangentMode=RCTM_Break, time=0.3, value=1}, <at 1 0>]}}})
   )
   &ScaleColor(
    Scale RGB=&VectorFromFloat(Value=&FloatFromCurve(FloatCurve=DataInterface() {propertyValues {Curve {keys [<at 0 2>,<at 0.5 1>,<at 1 0>]}}}))
    ScaleA=true
    ScaleRGB=true
   )
  ]
  rendererClasses [&NiagaraSpriteRendererProperties]
  renderers [{rendererClass=&NiagaraSpriteRendererProperties}]
 }
 IntroFlare {
  emitterUpdateScript [
   &EmitterState()
   &SpawnBurst_Instantaneous(Use Spawn Probability=false, UseLoopCountLimit=true)
  ]
  particleSpawnScript [
   &InitializeParticle(
    Color=<rgba 2 0.5254 0 1>
    Lifetime=0.35
    Sprite Size Mode=::NewEnumerator3
    Uniform Sprite Size=100
    UsePositionOffset=false
    Write Material Random=true
   )
   &CameraOffset(Camera Offset Amount=50)
  ]
  particleUpdateScript [
   &ParticleState()
   /SetVariables(Particles.Position=&Lerp_Vector(
     A=<0 0 175>
     B=&MakeCustomVectorFromBool(Bool=$User.LocalClient, True Vector=<0 0 238>, False Vector=<0 0 260>)
     Alpha=&FloatFromCurve(FloatCurve=DataInterface() {propertyValues {Curve {keys [
          {arriveTangent=-0.4775, interpMode=RCIM_Cubic, time=0, value=0}
          {arriveTangent=0.2037, interpMode=RCIM_Cubic, leaveTangent=0.2037, tangentMode=RCTM_User, time=0.5, value=1}
          {arriveTangent=-1.8374, interpMode=RCIM_Cubic, leaveTangent=-1.8374, tangentMode=RCTM_User, time=1, value=0.4}
         ]}}})
    )) {isSetParameters=true}
   &ScaleSpriteSize(
    Scale Sprite Size Mode=::NewEnumerator0
    Uniform Scale Factor=&FloatFromCurve(FloatCurve=DataInterface() {propertyValues {Curve {keys [
         {arriveTangent=7.1241, interpMode=RCIM_Cubic, leaveTangent=7.1241, tangentMode=RCTM_User, time=0, value=0.5}
         {arriveTangent=0.5, interpMode=RCIM_Cubic, leaveTangent=0.5, time=0.5, value=2}
         {arriveTangent=-2.1, tangentMode=RCTM_User, time=1, value=1}
        ]}}})
   )
   &ScaleColor(
    Scale Alpha=&FloatFromCurve(FloatCurve=DataInterface() {propertyValues {Curve {keys [
         {arriveTangent=25.9611, interpMode=RCIM_Cubic, leaveTangent=25.9611, tangentMode=RCTM_User, time=0, value=0}
         <at 0.1 1 RCIM_Cubic>
         {arriveTangent=-3.9853, tangentMode=RCTM_User, time=1, value=0}
        ]}}})
    ScaleA=true
    ScaleRGB=true
   )
  ]
  rendererClasses [&NiagaraSpriteRendererProperties]
  renderers [{rendererClass=&NiagaraSpriteRendererProperties}]
 }
 Light_Rays {
  emitterUpdateScript [
   &EmitterState()
   &SpawnRate(SpawnRate=10, Use Spawn Probability=false)
  ]
  particleSpawnScript [
   &InitializeParticle(
    Color=<rgba 2 1.1518 0 0.2>
    Lifetime Mode=::NewEnumerator1
    Mesh Scale Mode=::NewEnumerator3
    Mesh Uniform Scale=50
    Position Offset=&MakeCustomVectorFromBool(Bool=$User.LocalClient, True Vector=<0 0 205>, False Vector=<0 0 230>)
    Sprite Size Mode=::NewEnumerator1
    Uniform Sprite Size Max=1
    Uniform Sprite Size Min=1
    UsePositionOffset=true
    Write Material Random=true
   )
   &InitialMeshOrientation(
    Mesh Axis To Orient=<0 0 1>
    Mesh Orientation Mode=::NewEnumerator2
    Rotation=&RandomRangeVector(Minimum=<0 -0.0625 -0.0625>, Maximum=<0 0.0625 0.0625>)
    Use Rotation Vector=true
   )
  ]
  particleUpdateScript [
   &ParticleState()
   &DynamicMaterialParameters(
    Index 0 Param 1=&FloatFromCurve(
     FloatCurve=DataInterface() {propertyValues {Curve {keys [{interpMode=RCIM_Cubic, tangentMode=RCTM_User, time=0, value=0.5}, {interpMode=RCIM_Cubic, tangentMode=RCTM_User, time=0.5, value=1}]}}}
     Scale Curve=&RandomRangeFloat(Minimum=6, Maximum=12)
    )
    Index 0 Param 2=&RandomRangeFloat(Minimum=0, Maximum=1)
    Module.Index0Param0Write=true
    Module.Index0Param1Write=true
    Module.Index0Param2Write=true
    Module.Index0Param3Write=false
   )
   &ScaleColor(
    Scale Alpha=&FloatFromCurve(FloatCurve=DataInterface() {propertyValues {Curve {keys [
         {arriveTangent=9.0725, interpMode=RCIM_Cubic, leaveTangent=9.0725, tangentMode=RCTM_User, time=0, value=0}
         {arriveTangent=-2.5249, interpMode=RCIM_Cubic, leaveTangent=-2.5249, tangentMode=RCTM_User, time=0.1408, value=0.9847}
         {arriveTangent=-0.0405, interpMode=RCIM_Cubic, leaveTangent=-0.0405, tangentMode=RCTM_User, time=1, value=0}
        ]}}})
    ScaleA=true
    ScaleRGB=false
   )
   &ScaleMeshSize(Scale Factor=&MakeVector(
     X=&FloatFromCurve(FloatCurve=DataInterface() {propertyValues {Curve {keys [{interpMode=RCIM_Cubic, tangentMode=RCTM_User, time=0, value=0.5}, {interpMode=RCIM_Cubic, tangentMode=RCTM_User, time=1, value=1}]}}})
     Y=1
     Z=1
    )) {module=ScaleMeshSize001}
  ]
  rendererClasses [&NiagaraMeshRendererProperties]
  renderers [{rendererClass=&NiagaraMeshRendererProperties, rendererIndex=0}]
 }
 Smoke {
  emitterUpdateScript [
   &EmitterState()
   &SpawnParticlesFromOtherEmitter(
    Attribute Reader=DataInterface() {propertyValues {EmitterBinding {bindingMode=Other, emitterName=Smoke_S}}}
    Spawn Rate=12
    Use Maximum Age=false
    Use Minimum Age=false
   )
  ]
  particleSpawnScript [
   &InitializeParticle(
    Color=&RandomRangeLinearColor(Minimum=<rgba 0.4327 1.5042 5 0.45>, Maximum=<rgba 1.0964 0.8594 5 0.45>)
    Lifetime Max=1.3
    Lifetime Min=0.8
    Lifetime Mode=::NewEnumerator1
    Sprite Rotation Mode=::NewEnumerator2
    Sprite Size Mode=::NewEnumerator1
    Uniform Sprite Size Max=90
    Uniform Sprite Size Min=80
    UsePositionOffset=false
    Write Material Random=false
   )
   &SampleParticlesFromOtherEmitter(
    Applied Position Offset:Vector3f=<0 0 12>
    Material Random Sampling:ENiagara_DirectReadSamplingMode=::NewEnumerator0
    Maximum Sampled Particle Age=0.95
    Module.MaxSampledAgeCutoff=true
    Module.MinSampledAgeCutoff=false
    Reject Particle By Sampled Age=true
   )
   &CameraOffset(Camera Offset Amount=-5)
   &AddVelocity(Velocity=<0 0 10>)
  ]
  particleUpdateScript [
   &ParticleState()
   &InheritSourceMovement(
    Applied Position Scale=&VectorFromFloat(Value=&FloatFromCurve(FloatCurve=DataInterface() {propertyValues {Curve {keys [{interpMode=RCIM_Cubic, tangentMode=RCTM_User, time=0, value=1}, {interpMode=RCIM_Cubic, leaveTangent=0.097, time=1, value=0}]}}}))
    ApplyForce=false
    ApplyPosition=true
    ApplyPositionFalloff=false
   )
   &SolveForcesAndVelocity(Clamp Velocity=false, Limit Acceleration=false)
   &DynamicMaterialParameters(
    Index 0 Param 1=&MakeCustomFloatFromBool(Bool=&RandomBool(), True Float=0, False Float=0.75)
    Module.Param0WriteEnabled=true
    Module.Param1WriteEnabled=true
    Module.Param2WriteEnabled=true
    Module.Param3WriteEnabled=true
   )
   &ScaleColor(
    Scale Alpha=&FloatFromCurve(
     FloatCurve=DataInterface() {propertyValues {Curve {keys [
         {arriveTangent=16.2969, interpMode=RCIM_Cubic, leaveTangent=16.2969, tangentMode=RCTM_User, time=0, value=0}
         {interpMode=RCIM_Cubic, leaveTangent=-2.3559, tangentMode=RCTM_Break, time=0.2, value=1}
         {tangentMode=RCTM_User, time=1, value=0}
        ]}}}
     Scale Curve=0.7
    )
    ScaleA=true
    ScaleRGB=true
   )
   &SpriteRotationRate(Rotation Rate=&RandomRangeFloat(Minimum=-70, Maximum=70))
  ]
  rendererClasses [&NiagaraSpriteRendererProperties]
  renderers [{rendererClass=&NiagaraSpriteRendererProperties}]
 }
 SmokePuffs {
  emitterUpdateScript [
   &EmitterState(Life Cycle Mode=::NewEnumerator1, Loop Behavior=::NewEnumerator1, Loop Duration=3.6, UseLoopDelay=false)
   &SpawnRate(SpawnRate=18, Use Spawn Probability=false)
   &SpawnPerUnit(
    Max Movement Threshold=750
    Module.Use Max Movement Threshold=true
    Module.Use Movement Tolerance=false
    Spawn Group=1
    Spawn Spacing=50
    Use Spawn Probability=false
   )
  ]
  particleSpawnScript [
   &InitializeParticle(
    Color Maximum=<rgba 0.2193 0.1719 1 1>
    Color Minimum=<rgba 0.0865 0.3008 1 1>
    Color Mode=::NewEnumerator2
    Lifetime=&RandomRangeFloat(Minimum=1.4, Maximum=2)
    Mesh Scale Mode=::NewEnumerator1
    Mesh Uniform Scale Max=30
    Mesh Uniform Scale Min=10
    Sprite Rotation Mode=::NewEnumerator2
    Sprite Size Mode=::NewEnumerator3
    Uniform Sprite Size=60
    UsePositionOffset=true
    Write Material Random=true
   )
   &ShapeLocation(
    Cylinder Height=1
    Cylinder Height Midpoint=0
    Cylinder Radius=40
    MaskSpawnGroup=false
    Override Seed=false
    Shape Primitive=::NewEnumerator1
    Surface Only=true
    Use Endcaps In Surface Only Mode=false
    UseOwnerScale=true
   )
   &CameraOffset(Camera Offset Amount=25)
   &AddVelocity(
    Velocity=$Particles.ShapeLocation.ShapeNormal
    Velocity Speed Scale=&RandomRangeFloat(Minimum=0, Maximum=150)
   )
  ]
  particleUpdateScript [
   &ParticleState()
   &Drag(Drag=3, Use Linear Drag=true, Use Rotational Drag=true)
   &InheritSourceMovement(
    Applied Position Scale=&VectorFromFloat(Value=&MaskFloatBySpawnGroup(
      Masked Float=&FloatFromCurve(FloatCurve=DataInterface() {propertyValues {Curve {keys [{interpMode=RCIM_Cubic, tangentMode=RCTM_User, time=0, value=1}, {interpMode=RCIM_Cubic, tangentMode=RCTM_User, time=1, value=0.8}]}}})
      Passthrough Float=&FloatFromCurve(FloatCurve=DataInterface() {propertyValues {Curve {keys [<at 0 1>, {interpMode=RCIM_Cubic, tangentMode=RCTM_User, time=1, value=0.2}]}}})
     ))
    ApplyForce=false
    ApplyPosition=true
    ApplyPositionFalloff=false
   )
   &GravityForce(Gravity=<0 0 -100>)
   &SolveForcesAndVelocity(Clamp Velocity=false, Limit Acceleration=false)
   &SpriteRotationRate(Rotation Rate=25)
   &DynamicMaterialParameters(
    Index 0 Param 1=&Multiply_Float(
     A=&Add_Float(A=$Engine.RealTime, B=&RandomRangeFloat(Minimum=0, Maximum=10))
     B=0.1
    )
    Index 0 Param 2=&FloatFromCurve(FloatCurve=DataInterface() {propertyValues {Curve {keys [
         {interpMode=RCIM_Cubic, tangentMode=RCTM_User, time=0, value=1}
         {interpMode=RCIM_Cubic, tangentMode=RCTM_User, time=0.2, value=0.75}
         {arriveTangent=0.0015, interpMode=RCIM_Cubic, leaveTangent=0.0015, tangentMode=RCTM_User, time=0.7, value=0.75}
         {tangentMode=RCTM_User, time=1, value=1}
        ]}}})
    Index 2 Param 1=&FloatFromCurve(FloatCurve=DataInterface() {propertyValues {Curve {keys [
         {interpMode=RCIM_Cubic, tangentMode=RCTM_User, time=0, value=0}
         {arriveTangent=-0.1028, interpMode=RCIM_Cubic, leaveTangent=-0.1028, tangentMode=RCTM_User, time=1, value=2}
        ]}}})
    Index 2 Param 2=$Particles.MaterialRandom
    Index 2 Param 3=&FloatFromCurve(FloatCurve=DataInterface() {propertyValues {Curve {keys [
         {interpMode=RCIM_Cubic, tangentMode=RCTM_User, time=0, value=0}
         {arriveTangent=4.4069, interpMode=RCIM_Cubic, leaveTangent=4.4069, tangentMode=RCTM_User, time=1, value=2}
        ]}}})
    Index 2 Param 4=0.75
    Module.Index0Param0Write=true
    Module.Index0Param1Write=true
    Module.Index0Param2Write=true
    Module.Index0Param3Write=true
    Module.Index2Param0Write=true
    Module.Index2Param1Write=true
    Module.Index2Param2Write=true
    Module.Index2Param3Write=true
    Write Parameter Index 2=true
   )
   &ScaleSpriteSize(Uniform Curve Sprite Scale=DataInterface() {propertyValues {Curve {keys [{arriveTangent=0.8982, interpMode=RCIM_Cubic, leaveTangent=0.8982, tangentMode=RCTM_User, time=0, value=1}, <at 1 2 RCIM_Cubic>]}}})
   &ScaleColor(
    Scale Alpha=&FloatFromCurve(
     FloatCurve=DataInterface() {propertyValues {Curve {keys [
         {arriveTangent=9.0725, interpMode=RCIM_Cubic, leaveTangent=9.0725, tangentMode=RCTM_User, time=0, value=0}
         {arriveTangent=0.1111, interpMode=RCIM_Cubic, leaveTangent=0.1111, tangentMode=RCTM_User, time=0.1408, value=1}
         {arriveTangent=-3.3135, interpMode=RCIM_Cubic, leaveTangent=-3.3135, tangentMode=RCTM_User, time=1, value=0}
        ]}}}
     Scale Curve=0.25
    )
    ScaleA=true
    ScaleRGB=true
   )
  ]
  rendererClasses [&NiagaraSpriteRendererProperties]
  renderers [{rendererClass=&NiagaraSpriteRendererProperties}]
 }
 Smoke_S {
  emitterUpdateScript [
   &EmitterState(Life Cycle Mode=::NewEnumerator1, Loop Behavior=::NewEnumerator2, Loop Count=7, Loop Duration=0.5, UseLoopDelay=false)
   &SpawnPerUnit(
    Max Movement Threshold=750
    Module.Use Max Movement Threshold=true
    Module.Use Movement Tolerance=false
    Spawn Group=1
    Spawn Spacing=150
    Use Spawn Probability=false
   )
   &SpawnBurst_Instantaneous(Use Spawn Probability=false, UseLoopCountLimit=false)
  ]
  particleSpawnScript [
   &InitializeParticle(
    Color Maximum=<rgba 0.4386 0.3438 2 1>
    Color Minimum=<rgba 0.1731 0.6017 2 1>
    Color Mode=::NewEnumerator2
    Lifetime Max=2.25
    Lifetime Min=1.2
    Lifetime Mode=::NewEnumerator1
    Material Random:NiagaraFloat=&MaskFloatBySpawnGroup(Masked Float=0, Passthrough Float=1)
    Position Offset=<0 0 2>
    Position Offset Coordinate Space=::World
    Sprite Size Max=<1 1>
    Sprite Size Min=<0.8 0.8>
    Sprite Size Mode=::NewEnumerator2
    UsePositionOffset=true
    Write Material Random=true
   )
   &CameraOffset(Camera Offset Amount=-18)
   &ShapeLocation(
    Cylinder Height=1
    Cylinder Height Midpoint=0
    Cylinder Radius=80
    MaskSpawnGroup=false
    Offset=<0 0 13>
    Offset Coordinate Space=::World
    Offset Mode=::NewEnumerator0
    Override Seed=false
    Shape Primitive=::NewEnumerator1
    Surface Only=false
    UseOwnerScale=true
   )
   &AddVelocity(Rotation Coordinate Space=::World, Velocity=&MakeVector(X=0, Y=0, Z=45))
  ]
  particleUpdateScript [
   &ParticleState()
   &Drag(Drag=0.25, Use Linear Drag=true, Use Rotational Drag=true)
   &InheritSourceMovement(
    Applied Position Scale=&VectorFromFloat(Value=&FloatFromCurve(FloatCurve=DataInterface() {propertyValues {Curve {keys [{interpMode=RCIM_Cubic, tangentMode=RCTM_User, time=0.2, value=1}, {interpMode=RCIM_Cubic, leaveTangent=0.0969, time=1, value=0}]}}}))
    ApplyForce=false
    ApplyPosition=true
    ApplyPositionFalloff=false
   )
   &SolveForcesAndVelocity(Clamp Velocity=false, Limit Acceleration=false)
   &ScaleSpriteSize(
    Uniform Curve Scale=75
    Uniform Curve Sprite Scale=DataInterface() {propertyValues {Curve {keys [
        {arriveTangent=0.8382, interpMode=RCIM_Cubic, leaveTangent=0.8382, tangentMode=RCTM_User, time=0, value=0.35}
        {arriveTangent=0.0224, interpMode=RCIM_Cubic, leaveTangent=0.0224, tangentMode=RCTM_User, time=1, value=1}
       ]}}}
   )
   &ScaleSpriteSizeBySpeed(Max Scale Factor=<0.25 2>, Velocity Threshold=400)
   &DynamicMaterialParameters(
    Index 0 Param 2=&FloatFromCurve(
     FloatCurve=DataInterface() {propertyValues {Curve {keys [
         {arriveTangent=0.1994, interpMode=RCIM_Cubic, leaveTangent=0.1994, tangentMode=RCTM_User, time=0, value=1}
         {
          arriveTangent=0.128
          arriveTangentWeight=0.079
          interpMode=RCIM_Cubic
          leaveTangent=0.1794
          leaveTangentWeight=0.4542
          tangentMode=RCTM_Break
          tangentWeightMode=RCTWM_WeightedBoth
          time=0.2
          value=0.15
         }
         {arriveTangent=2.4303, interpMode=RCIM_Cubic, leaveTangent=2.4303, tangentMode=RCTM_User, time=1, value=1}
        ]}}}
     Scale Curve=&RandomRangeFloat(Minimum=0.6, Maximum=0.7)
    )
    Index 0 Param 3=&FloatFromCurve(
     FloatCurve=DataInterface() {propertyValues {Curve {keys [
         {arriveTangent=-0.2848, interpMode=RCIM_Cubic, time=0, value=1}
         {arriveTangent=-0.2044, leaveTangent=-0.0269, tangentMode=RCTM_User, time=0.3, value=0.1}
         {arriveTangent=1.2857, interpMode=RCIM_Cubic, leaveTangent=1.2857, time=0.5, value=0.1}
         {arriveTangent=1.6876, interpMode=RCIM_Cubic, leaveTangent=1.6876, tangentMode=RCTM_User, time=1, value=1}
        ]}}}
     Scale Curve=0.4
    )
    Index 0 Param 4=&RandomRangeFloat(Minimum=0, Maximum=1)
    Index 1 Param 3=&RandomRangeFloat(Minimum=0.1, Maximum=1)
    Index 1 Param 4=&FloatFromCurve(
     FloatCurve=DataInterface() {propertyValues {Curve {keys [
         {arriveTangent=9.0725, interpMode=RCIM_Cubic, leaveTangent=9.0725, tangentMode=RCTM_User, time=0, value=1}
         {arriveTangent=-2.5249, interpMode=RCIM_Cubic, leaveTangent=-2.5249, tangentMode=RCTM_User, time=0.1, value=2}
         {arriveTangent=-1.1111, interpMode=RCIM_Cubic, leaveTangent=-1.1111, time=0.3, value=1.25}
         {arriveTangent=-0.0405, interpMode=RCIM_Cubic, leaveTangent=-0.0405, tangentMode=RCTM_User, time=1, value=1}
        ]}}}
     Scale Curve=&Add_Float(
      A=1
      B=&Add_Float(
       A=&FloatFromCurve(
        FloatCurve=DataInterface() {propertyValues {Curve {keys [
            {arriveTangent=-25.6702, interpMode=RCIM_Cubic, leaveTangent=-25.6702, tangentMode=RCTM_User, time=0, value=10}
            {arriveTangent=-0.937, interpMode=RCIM_Cubic, leaveTangent=0.4276, tangentMode=RCTM_Break, time=1, value=0.05}
            <at 1.3 0.05>
            <at 1.35 1 RCIM_Cubic>
            <at 1.8 0.05 RCIM_Cubic>
            <at 1.85 1 RCIM_Cubic>
            <at 2.25 0.05 RCIM_Cubic>
            <at 2.3 1 RCIM_Cubic>
            {arriveTangent=-1.5625, leaveTangent=-1.5625, time=2.7, value=0.05}
            {arriveTangent=3.8, interpMode=RCIM_Cubic, leaveTangent=3.8, time=2.9, value=0.05}
            <at 2.95 1 RCIM_Cubic>
            {arriveTangent=-1.1111, leaveTangent=-1.1111, time=3.35, value=0.05}
            {arriveTangent=1.5833, interpMode=RCIM_Cubic, leaveTangent=1.5833, time=3.9, value=0.05}
            <at 3.95 1 RCIM_Cubic>
            <at 4.35 0.05 RCIM_Cubic>
            <at 4.4 1 RCIM_Cubic>
            {arriveTangent=-2, leaveTangent=-2, time=4.8, value=0.05}
            {arriveTangent=6.3333, interpMode=RCIM_Cubic, leaveTangent=6.3333, time=4.9, value=0.05}
            <at 4.95 1 RCIM_Cubic>
           ]}}}
        CurveIndex=$Emitter.Age
        Scale Curve=6
       )
       B=&FloatFromCurve(
        FloatCurve=DataInterface() {propertyValues {Curve {keys [<at 0.95 0 RCIM_Cubic>,<at 1 1 RCIM_Cubic>,<at 2.1 0 RCIM_Cubic>,<at 2.15 1 RCIM_Cubic>,<at 3.05 0 RCIM_Cubic>,<at 3.1 1 RCIM_Cubic>,<at 5 0 RCIM_Cubic>]}}}
        CurveIndex=$Emitter.Age
       )
      )
     )
    )
    Module.Index0Param0Write=true
    Module.Index0Param1Write=true
    Module.Index0Param2Write=true
    Module.Index0Param3Write=true
    Module.Index1Param0Write=true
    Module.Index1Param1Write=true
    Module.Index1Param2Write=true
    Module.Index1Param3Write=true
    Module.Index2Param0Write=true
    Module.Index2Param1Write=true
    Module.Index2Param2Write=true
    Module.Index2Param3Write=true
    Module.Index3Param0Write=true
    Module.Index3Param1Write=true
    Module.Index3Param2Write=true
    Module.Index3Param3Write=true
    Write Parameter Index 1=true
    Write Parameter Index 2=true
    Write Parameter Index 3=true
   )
   &ScaleColor(
    Scale Alpha=&FloatFromCurve(
     FloatCurve=DataInterface() {propertyValues {Curve {keys [
         {arriveTangent=0.7526, interpMode=RCIM_Cubic, time=0, value=0}
         {arriveTangent=4.8353, leaveTangent=-0.8276, tangentMode=RCTM_User, time=0.3, value=1}
         {arriveTangent=-1.4286, interpMode=RCIM_Cubic, leaveTangent=-1.4286, time=0.475, value=1}
         {interpMode=RCIM_Cubic, leaveTangent=-1.5738, time=1, value=0}
        ]}}}
     Scale Curve=&FloatFromCurve(FloatCurve=DataInterface() {propertyValues {Curve {keys [<at 4.8 1>,<at 5 0>]}}}, CurveIndex=$Emitter.Age, Scale Curve=0.6)
    )
    ScaleA=true
    ScaleRGB=true
   )
  ]
  rendererClasses [&NiagaraSpriteRendererProperties]
  renderers [{rendererClass=&NiagaraSpriteRendererProperties}]
 }
 Spline_Light_Rays {
  emitterUpdateScript [
   &EmitterState(Delay First Loop Only=true, Life Cycle Mode=::NewEnumerator1, Loop Delay=0.45, Loop Duration=0.45, UseLoopDelay=true)
   &SpawnBurst_Instantaneous(Spawn Count=2, Use Spawn Probability=false, UseLoopCountLimit=false)
  ]
  particleSpawnScript [
   /SetVariables(Particles.SpawnRotationVector=&RandomRangeVector(Minimum=<-0.05 -0.05 -0.35>, Maximum=<0.05 0.05 -0.25>, Evaluation Type=::NewEnumerator1)) {isSetParameters=true}
   &InitializeParticle(
    Color=<rgba 8 3.7468 0 0.2>
    Lifetime=0.55
    Mesh Scale Mode=::NewEnumerator1
    Mesh Uniform Scale Max=0.7
    Mesh Uniform Scale Min=0.5
    Position Offset=&MakeCustomVectorFromBool(Bool=$User.LocalClient, True Vector=<0 0 205>, False Vector=<0 0 230>)
    UsePositionOffset=true
    Write Material Random=true
   )
   &CameraOffset(Camera Offset Amount=10)
   &DynamicMaterialParameters(
    Index 0 Param 1=&RandomRangeFloat(Minimum=1, Maximum=2, Evaluation Type=::NewEnumerator1)
    Index 0 Param 2=&RandomRangeFloat(Minimum=0, Maximum=1, Evaluation Type=::NewEnumerator1)
    Index 0 Param 3=&RandomRangeFloat(Minimum=2, Maximum=3, Evaluation Type=::NewEnumerator1)
    Module.Index0Param0Write=true
    Module.Index0Param1Write=true
    Module.Index0Param2Write=true
    Module.Index0Param3Write=false
   )
   &InitialMeshOrientation(Use Rotation Vector=false)
  ]
  particleUpdateScript [
   &ParticleState()
   &UpdateMeshOrientation(Rotation Vector=$Particles.SpawnRotationVector)
   &ScaleMeshSize(Scale Factor=&VectorFromFloat(Value=&FloatFromCurve(FloatCurve=DataInterface() {propertyValues {Curve {keys [{arriveTangent=1.5984, interpMode=RCIM_Cubic, time=0, value=0.5}, {interpMode=RCIM_Cubic, leaveTangent=0.41, time=1, value=1.25}]}}})))
   &ScaleColor(
    Scale Alpha=&FloatFromCurve(FloatCurve=DataInterface() {propertyValues {Curve {keys [{arriveTangent=26.35, interpMode=RCIM_Cubic, time=0, value=0}, <at 0.1 1 RCIM_Cubic>, <at 1 0 RCIM_Cubic>]}}})
    Scale RGB=&VectorFromFloat(Value=&Add_Float(
      A=&FloatFromCurve(
       FloatCurve=DataInterface() {propertyValues {Curve {keys [
           {arriveTangent=-38.4417, time=0, value=0.05}
           <at 1.3 0.05>
           <at 1.35 1 RCIM_Cubic>
           <at 1.8 0.05 RCIM_Cubic>
           <at 1.85 1 RCIM_Cubic>
           <at 2.25 0.05 RCIM_Cubic>
           <at 2.3 1 RCIM_Cubic>
           {arriveTangent=-1.5625, leaveTangent=-1.5625, time=2.7, value=0.05}
           {arriveTangent=3.8, interpMode=RCIM_Cubic, leaveTangent=3.8, time=2.9, value=0.05}
           <at 2.95 1 RCIM_Cubic>
           {arriveTangent=-1.1111, leaveTangent=-1.1111, time=3.35, value=0.05}
           {arriveTangent=1.5833, interpMode=RCIM_Cubic, leaveTangent=1.5833, time=3.9, value=0.05}
           <at 3.95 1 RCIM_Cubic>
           <at 4.35 0.05 RCIM_Cubic>
           <at 4.4 1 RCIM_Cubic>
           {arriveTangent=-2, leaveTangent=-2, time=4.8, value=0.05}
           {arriveTangent=6.3333, interpMode=RCIM_Cubic, leaveTangent=6.3333, time=4.9, value=0.05}
           <at 4.95 1 RCIM_Cubic>
          ]}}}
       CurveIndex=$Emitter.Age
       Scale Curve=2
      )
      B=1
     ))
    ScaleA=true
    ScaleRGB=true
   )
  ]
  rendererClasses [&NiagaraMeshRendererProperties]
  renderers [{rendererClass=&NiagaraMeshRendererProperties, rendererIndex=0}]
 }
}
summary {
 systemName=NS_LemonCart_DanceGrenade_Loop
 emitters [
  {
   emitterName=DiscoBall
   rendererClasses [
    &NiagaraMeshRendererProperties
    &NiagaraLightRendererProperties
    &NiagaraComponentRendererProperties
   ]
  }
  {emitterName=DiscoBall_Sparkles, rendererClasses [&NiagaraSpriteRendererProperties]}
  {emitterName=Spline_Light_Rays, rendererClasses [&NiagaraMeshRendererProperties]}
  {emitterName=Corona_And_Spread, rendererClasses [&NiagaraSpriteRendererProperties]}
  {emitterName=IntroFlare, rendererClasses [&NiagaraSpriteRendererProperties]}
  {emitterName=Light_Rays, rendererClasses [&NiagaraMeshRendererProperties]}
  {emitterName=SmokePuffs, rendererClasses [&NiagaraSpriteRendererProperties]}
  {emitterName=Smoke_S, rendererClasses [&NiagaraSpriteRendererProperties]}
  {emitterName=Smoke, rendererClasses [&NiagaraSpriteRendererProperties]}
 ]
 userVariables [
  User.FXAlpha {type=NiagaraFloat, defaultValue=0, description=""}
  User.LocalClient {type=NiagaraBool}
 ]
}
